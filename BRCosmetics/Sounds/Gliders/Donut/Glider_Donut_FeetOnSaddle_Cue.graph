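# Sound cue for the Donut glider's feet-on-saddle sound, kept with the Battle Royale cosmetic glider audio.
# A single wave player feeds the output directly, playing the Glider_Donut_Open_FS wave with no randomization or modulation.
# The cue uses the shared wide-pan glider open/close attenuation and concurrency, is routed to the glider land sound class, and plays at reduced volume.
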SoundCue /BRCosmetics/Sounds/Gliders/Donut/Glider_Donut_FeetOnSaddle_Cue
root=WavePlayer_0
nodes [{id=WavePlayer_0, properties {soundWaveAssetPtr=&/Glider_Donut_Open_FS}}]
properties {
 attenuationSettings=/BRCosmetics/Sounds/Other/Attenuation/Fort_Player_Glider_OpenClose_Athena_WidePan_Att
 maxDistance=10000
 soundClassObject=&Fort_Gliders_Land
 volumeMultiplier=0.6
 concurrencySet [&Athena_GliderOpenClose_Conc]
}
