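# Sound cue for the Donut glider's gallop, a cosmetic glider sound that layers gallop and thrust audio.
# A Mixer plays two equally weighted Random branches together at heavily reduced volume, one choosing among three gallop waves, the other among three thrust waves.
# It uses glider thrust attenuation and the glider land sound class.
# A concurrency override caps simultaneous plays per owner, stopping the oldest, in place of the assigned open/close concurrency set.
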SoundCue /BRCosmetics/Sounds/Gliders/Donut/Glider_Donut_Gallop
root=Mixer_0
nodes [
 {id=Mixer_0, children [Random_0,Random_1], properties {inputVolume [0.1,0.1]}}
 {id=Random_0, children [WavePlayer_1,WavePlayer_3,WavePlayer_4], properties {weights [1,1,1]}}
 {id=Random_1, children [WavePlayer_5,WavePlayer_6,WavePlayer_8], properties {weights [1,1,1]}}
 {id=WavePlayer_1, properties {soundWaveAssetPtr=&/Glider_Donut_Gallop_01}}
 {id=WavePlayer_3, properties {soundWaveAssetPtr=&/Glider_Donut_Gallop_02}}
 {id=WavePlayer_4, properties {soundWaveAssetPtr=&/Glider_Donut_Gallop_03}}
 {id=WavePlayer_5, properties {soundWaveAssetPtr=&/Glider_Donut_Thrust_04}}
 {id=WavePlayer_6, properties {soundWaveAssetPtr=&/Glider_Donut_Thrust_01}}
 {id=WavePlayer_8, properties {soundWaveAssetPtr=&/Glider_Donut_Thrust_03}}
]
properties {
 attenuationSettings=&Fort_Player_Glider_Thrust_Athena_Att
 bOverrideConcurrency=true
 maxDistance=12500
 soundClassObject=&Fort_Gliders_Land
 volumeMultiplier=1
 concurrencyOverrides {bLimitToOwner=true, maxCount=2, resolutionRule=StopOldest}
 concurrencySet [&Athena_GliderOpenClose_Conc]
}
