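# Sound cue for the Donut glider's opening sound, layering the open effect with a delayed llama sound.
# A mixer plays the open wave alongside a quieter branch, which waits a fixed one second before picking one of four equally weighted llama variations at random.
# Glider-specific attenuation, concurrency, sound class and property sheet settings govern spatialization, voice limits and mixing.
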SoundCue /BRCosmetics/Sounds/Gliders/Donut/Glider_Open_Donut
root=Mixer_0
nodes [
 {id=Delay_0, children [Random_0], properties {delayMax=1, delayMin=1}}
 {id=Mixer_0, children [WavePlayer_1,Delay_0], properties {inputVolume [1,0.7]}}
 {id=Random_0, children [WavePlayer_0,WavePlayer_2,WavePlayer_3,WavePlayer_4], properties {weights [1,1,1,1]}}
 {id=WavePlayer_0, properties {soundWaveAssetPtr=&/Glider_Donut_Llama_01}}
 {id=WavePlayer_1, properties {soundWaveAssetPtr=&/Glider_Donut_Open}}
 {id=WavePlayer_2, properties {soundWaveAssetPtr=&/Glider_Donut_Llama_02}}
 {id=WavePlayer_3, properties {soundWaveAssetPtr=&/Glider_Donut_Llama_03}}
 {id=WavePlayer_4, properties {soundWaveAssetPtr=&/Glider_Donut_Llama_04}}
]
properties {
 attenuationSettings=/BRCosmetics/Sounds/Other/Attenuation/Fort_Player_Glider_OpenClose_Athena_WidePan_Att
 audioPropertiesSheet=/BRCosmetics/Sounds/Settings/PropertySheets/Glider/APS_Cosmetic_Glider_Open
 maxDistance=10000
 soundClassObject=&Fort_Gliders_Land
 volumeMultiplier=0.7
 concurrencySet [&Athena_GliderOpenClose_Conc]
}
