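# Sound cue for the Donut glider's thrust-start sound in Battle Royale.
# A Mixer layers two equally weighted Random branches, one picking from four thrust waves and the other from four llama waves, with the llama layer mixed quieter.
# It uses the player glider thrust attenuation and a shared glider thrust concurrency set, and is routed to the Fort_Gliders_Land sound class.
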SoundCue /BRCosmetics/Sounds/Gliders/Donut/Glider_ThrustStart_Donut
root=Mixer_0
nodes [
 {id=Mixer_0, children [Random_1,Random_2], properties {inputVolume [0.3,0.2]}}
 {id=Random_1, children [WavePlayer_0,WavePlayer_2,WavePlayer_7,WavePlayer_8], properties {weights [1,1,1,1]}}
 {id=Random_2, children [WavePlayer_12, WavePlayer_13, WavePlayer_14, WavePlayer_15], properties {weights [1,1,1,1]}}
 {id=WavePlayer_0, properties {soundWaveAssetPtr=&/Glider_Donut_Thrust_01}}
 {id=WavePlayer_12, properties {soundWaveAssetPtr=&/Glider_Donut_Llama_01}}
 {id=WavePlayer_13, properties {soundWaveAssetPtr=&/Glider_Donut_Llama_02}}
 {id=WavePlayer_14, properties {soundWaveAssetPtr=&/Glider_Donut_Llama_03}}
 {id=WavePlayer_15, properties {soundWaveAssetPtr=&/Glider_Donut_Llama_04}}
 {id=WavePlayer_2, properties {soundWaveAssetPtr=&/Glider_Donut_Thrust_02}}
 {id=WavePlayer_7, properties {soundWaveAssetPtr=&/Glider_Donut_Thrust_03}}
 {id=WavePlayer_8, properties {soundWaveAssetPtr=&/Glider_Donut_Thrust_04}}
]
properties {
 attenuationSettings=&Fort_Player_Glider_Thrust_Athena_Att
 maxDistance=12500
 soundClassObject=&Fort_Gliders_Land
 volumeMultiplier=1
 concurrencyOverrides {bLimitToOwner=true, maxCount=2, resolutionRule=StopOldest}
 concurrencySet [&Athena_Glider_Thrust_Conc]
}
